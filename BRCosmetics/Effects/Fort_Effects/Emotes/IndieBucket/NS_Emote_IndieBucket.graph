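# Niagara system for the IndieBucket emote that shows eight long-lived icon particles, rendered as meshes and sprites, pinned to skeletal mesh bones or sockets.
# The Icons emitter uses each particle ID to pick its bone and one of eight timed Carrot curves.
# Those curves drive sprite size, mesh scale and a dynamic material parameter, while meshes orient horizontally relative to the emitter's vertical axis.
# The Fountain sprite emitter is disabled.
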NiagaraSystem /BRCosmetics/Effects/Fort_Effects/Emotes/IndieBucket/NS_Emote_IndieBucket
dependencies {
 usedDataInterfaces [
  &NiagaraDataInterfaceMeshRendererInfo
  &NiagaraDataInterfaceArrayColor
  &NiagaraDataInterfaceCurve
  &NiagaraDataInterfaceSkeletalMesh
 ]
 usedDynamicInputs [&Divide_Float]
 usedModules [
  &self:SystemSpawnScript.SystemScriptSource.SystemScriptGraph.NiagaraNodeAssignment_2.SetVariables_#1
  &self:SystemSpawnScript.SystemScriptSource.SystemScriptGraph.NiagaraNodeAssignment_0.SetVariables_#2
  &self:SystemSpawnScript.SystemScriptSource.SystemScriptGraph.NiagaraNodeAssignment_12.SetVariables_#3
  &self:SystemSpawnScript.SystemScriptSource.SystemScriptGraph.NiagaraNodeAssignment_11.SetVariables_#4
  &self:SystemSpawnScript.SystemScriptSource.SystemScriptGraph.NiagaraNodeAssignment_10.SetVariables_#5
  &self:SystemSpawnScript.SystemScriptSource.SystemScriptGraph.NiagaraNodeAssignment_9.SetVariables_#6
  &self:SystemSpawnScript.SystemScriptSource.SystemScriptGraph.NiagaraNodeAssignment_8.SetVariables_#7
  &self:SystemSpawnScript.SystemScriptSource.SystemScriptGraph.NiagaraNodeAssignment_7.SetVariables_#8
  &self:SystemSpawnScript.SystemScriptSource.SystemScriptGraph.NiagaraNodeAssignment_6.SetVariables_#9
  &SystemState
  &self:SystemSpawnScript.SystemScriptSource.SystemScriptGraph.NiagaraNodeAssignment_13.SetVariables_#10
 ]
}
emitters {
 Fountain {
  bEnabled=false
  emitterUpdateScript [
   &EmitterState(Life Cycle Mode=::NewEnumerator1, Loop Duration=2, UseLoopDelay=false)
   &SpawnRate(SpawnRate=90, Use Spawn Probability=false)
  ]
  particleSpawnScript [
   &InitializeParticle(
    Lifetime Max=1.75
    Lifetime Min=1.4
    Lifetime Mode=::NewEnumerator1
    Mass Mode=::NewEnumerator3
    Sprite Size Mode=::NewEnumerator3
    Uniform Sprite Size=100
    UsePositionOffset=false
    Write Material Random=true
   )
   &ShapeLocation(MaskSpawnGroup=false, Override Seed=false, Sphere Radius=0, UseOwnerScale=true)
  ]
  particleUpdateScript [
   &ParticleState()
   &DynamicMaterialParameters(
    Index 0 Param 1=$System.LoopedAge
    Module.Index0Param0Write=true
    Module.Index0Param1Write=true
    Module.Index0Param2Write=true
    Module.Index0Param3Write=true
   )
  ]
  rendererClasses [&NiagaraSpriteRendererProperties]
  renderers [{rendererClass=&NiagaraSpriteRendererProperties}]
 }
 Icons {
  emitterUpdateScript [
   &EmitterState()
   &SpawnBurst_Instantaneous(Spawn Count=8, Use Spawn Probability=false, UseLoopCountLimit=false)
  ]
  particleSpawnScript [
   &InitializeParticle(
    Color=<rgba 1 1 1 0.05>
    Lifetime Max=30
    Lifetime Mode=::NewEnumerator1
    Mesh Renderer Array Visibility Mode=::NewEnumerator1
    Mesh Renderer Info=DataInterface() {propertyValues {MeshRenderer=&self:Icons_0.NiagaraMeshRendererProperties_0}}
    Mesh Renderer Mesh Index:NiagaraInt32=&RandomRangeInt(Minimum=1, Maximum=1)
    Mesh Scale Mode=::NewEnumerator3
    Mesh Uniform Scale=0.05
    Sprite Size Mode=::NewEnumerator3
    Uniform Sprite Size=6
    UsePositionOffset=false
    Write Material Random=true
   )
   &Color(Color=&SelectLinearColorFromArray(
     Array Sampling Mode=::NewEnumerator1
     Direct Array Index=$Particles.MeshIndex
     Color Selection Array=DataInterface() {propertyValues {ColorData [<rgba 1 1 1 1>, <rgba 0.6 0.6 0.6 1>, <rgba 1 1 1 1>, <rgba 1 1 1 1>]}}
     Random Seed=?
    ))
   /SetVariables(Particles.Linear Color=&SelectLinearColorFromArray(
     Array Sampling Mode=::NewEnumerator1
     Direct Array Index=$Particles.MeshIndex
     Color Selection Array=DataInterface() {propertyValues {ColorData [<rgba 0.5 1 0.5 0.35>, <rgba 1 1 0 0.355>, <rgba 0.8833 0.65 1 0.5>, <rgba 0.8833 0 1 0.5>]}}
     Random Seed=?
    )) {isSetParameters=true}
   /SetVariables(Particles.cam off=25) {isSetParameters=true}
   &InitialMeshOrientation(
    Mesh Orientation Mode=::NewEnumerator7
    Rotation=&RandomRangeVector(Minimum=<0 1 0>, Maximum=<0 0 0>)
    Use Rotation Vector=true
   )
   &InitialMeshOrientation(
    Mesh Orientation Mode=::NewEnumerator7
    Rotation=&RandomRangeVector(Minimum=<0 0 0.25>, Maximum=<0 0 0.25>)
    Use Rotation Vector=true
   ) {module=InitialMeshOrientation001}
  ]
  particleUpdateScript [
   &ParticleState()
   &ScaleSpriteSize(
    Scale Sprite Size Mode=::NewEnumerator0
    Uniform Scale Factor=&Multiply_Float(
     A=&self:ScratchDynamicInput(
      Curve:NiagaraInt32=&ReturnParticleIDAttributes()
      Curve for Floats:NiagaraDataInterfaceCurve=$System.Carrot_01
      Curve for Floats001:NiagaraDataInterfaceCurve=$System.Carrot_02
      Curve for Floats002:NiagaraDataInterfaceCurve=$System.Carrot_03
      Curve for Floats003:NiagaraDataInterfaceCurve=$System.Carrot_04
      Curve for Floats004:NiagaraDataInterfaceCurve=$System.Carrot_05
      Curve for Floats005:NiagaraDataInterfaceCurve=$System.Carrot_06
      Curve for Floats006:NiagaraDataInterfaceCurve=$System.Carrot_07
      Curve for Floats007:NiagaraDataInterfaceCurve=$System.Carrot_08
      Sample:NiagaraFloat=&Add_Float(A=$Engine.System.Age, B=0.43)
     )
     B=10
    )
   )
   &SkeletalMeshLocation(
    Bone / Socket Index=&ReturnParticleIDAttributes()
    Bone Sampling Mode=::NewEnumerator4
    MaskSpawnGroup=false
    Orientation Sampling=::NewEnumerator1
    Skeletal Mesh=$System.Skeletal Mesh
   )
   &UpdateMeshOrientation(
    Facing Interpolation=false
    Facing Position:NiagaraPosition=&MakeVector(
     X=&MakeFloatFromVector(Vector=Linked() {linkedVariable {name=Engine.Emitter.SimulationPosition, type=NiagaraPosition}}, Channel=::NewEnumerator0)
     Y=&MakeFloatFromVector(Vector=Linked() {linkedVariable {name=Engine.Emitter.SimulationPosition, type=NiagaraPosition}}, Channel=::NewEnumerator1)
     Z=&MakeFloatFromVector(Vector=Linked() {linkedVariable {name=Particles.Position, type=NiagaraPosition}}, Channel=::NewEnumerator2)
    )
    Orientation Method=::NewEnumerator2
    Override Quat=false
    TransformFacingOrigin=false
    TransformFacingPosition=false
    TransformUp=false
   )
   &ScaleMeshSize(Scale Factor=&VectorFromFloat(Value=&Multiply_Float(
      A=&self:ScratchDynamicInput(
       Curve:NiagaraInt32=&ReturnParticleIDAttributes()
       Curve for Floats:NiagaraDataInterfaceCurve=$System.Carrot_01
       Curve for Floats001:NiagaraDataInterfaceCurve=$System.Carrot_02
       Curve for Floats002:NiagaraDataInterfaceCurve=$System.Carrot_03
       Curve for Floats003:NiagaraDataInterfaceCurve=$System.Carrot_04
       Curve for Floats004:NiagaraDataInterfaceCurve=$System.Carrot_05
       Curve for Floats005:NiagaraDataInterfaceCurve=$System.Carrot_06
       Curve for Floats006:NiagaraDataInterfaceCurve=$System.Carrot_07
       Curve for Floats007:NiagaraDataInterfaceCurve=$System.Carrot_08
       Sample:NiagaraFloat=&Add_Float(A=$Engine.System.Age, B=0.43)
      )
      B=8
     )))
   &SolveForcesAndVelocity(Clamp Velocity=false, Limit Acceleration=false)
   &DynamicMaterialParameters(
    Index 0 Param 1=&RandomRangeFloat(Minimum=0, Maximum=1)
    Index 0 Param 2=&Multiply_Float(
     A=&self:ScratchDynamicInput(
      Curve:NiagaraInt32=&ReturnParticleIDAttributes()
      Curve for Floats:NiagaraDataInterfaceCurve=$System.Carrot_01
      Curve for Floats001:NiagaraDataInterfaceCurve=$System.Carrot_02
      Curve for Floats002:NiagaraDataInterfaceCurve=$System.Carrot_03
      Curve for Floats003:NiagaraDataInterfaceCurve=$System.Carrot_04
      Curve for Floats004:NiagaraDataInterfaceCurve=$System.Carrot_05
      Curve for Floats005:NiagaraDataInterfaceCurve=$System.Carrot_06
      Curve for Floats006:NiagaraDataInterfaceCurve=$System.Carrot_07
      Curve for Floats007:NiagaraDataInterfaceCurve=$System.Carrot_08
      Sample:NiagaraFloat=&Lerp_Float(
       A=0.433
       B=11.733
       Alpha=&FloatFromCurve(
        FloatCurve=DataInterface() {propertyValues {Curve {keys [{tangentMode=RCTM_User, time=0, value=0}, {tangentMode=RCTM_User, time=1, value=1}]}}}
        CurveIndex=$System.NormalizedLoopAge
       )
      )
     )
     B=0.5
    )
    Module.Index0Param0Write=true
    Module.Index0Param1Write=true
    Module.Index0Param2Write=true
    Module.Index0Param3Write=true
   )
   /SetVariables() {isSetParameters=true}
   /SetVariables(Particles.debug:Vector4f=&MakeVector4(
     X=&MakeFloatFromInt(Int=&ReturnParticleIDAttributes())
     Y=0
     Z=0
     W=0
    )) {isSetParameters=true}
   &ScaleColor(
    Scale RGB=&VectorFromFloat(Value=&FNTimeOfDayScalar(Night Scale=1, Morning Scale=1, Evening Scale=1))
    ScaleA=true
    ScaleRGB=true
   ) {enabled=false}
  ]
  rendererClasses [&NiagaraMeshRendererProperties, &NiagaraSpriteRendererProperties]
  renderers [
   {rendererClass=&NiagaraMeshRendererProperties, rendererIndex=0}
   {rendererClass=&NiagaraSpriteRendererProperties, rendererIndex=1}
   {rendererClass=&NiagaraSpriteRendererProperties, rendererIndex=2}
  ]
 }
}
summary {
 systemName=NS_Emote_IndieBucket
 emitters [
  {emitterName=Icons, rendererClasses [&NiagaraMeshRendererProperties, &NiagaraSpriteRendererProperties]}
  {bEnabled=false, emitterName=Fountain, rendererClasses [&NiagaraSpriteRendererProperties]}
 ]
 userVariables [User.NormalizedNotifyProgress {type=NiagaraFloat, defaultValue=0, description=""}]
}
