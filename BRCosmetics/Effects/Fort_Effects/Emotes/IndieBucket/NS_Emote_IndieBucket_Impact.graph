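# Niagara system for the IndieBucket emote's impact: a short, pixel-glitch burst in yellow, orange and red, despite "pink" emitter names.
# Mesh boxes, sprite pixel ring and flash, flying glitch squares and a flare make up the live layers.
# User "Is Swing Left" zeroes spawning for boxes and flare; time of day and camera distance scale some layers.
# Six other emitters, including sparks and holographic center, are disabled.
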NiagaraSystem /BRCosmetics/Effects/Fort_Effects/Emotes/IndieBucket/NS_Emote_IndieBucket_Impact
dependencies {
 usedDataInterfaces [
  &NiagaraDataInterfaceCurve
  &NiagaraDataInterfaceArrayColor
  &NiagaraDataInterfaceColorCurve
  &NiagaraDataInterfaceArrayFloat
  &NiagaraDataInterfaceVector2DCurve
  &NiagaraDataInterfaceSkeletalMesh
 ]
 usedDynamicInputs [
  &SimulationPosition
  &SetBoolByFloatComparison
  &FN_Select_Float_from_Time_Of_Day
 ]
 usedModules [
  &self:SystemSpawnScript.SystemScriptSource.SystemScriptGraph.NiagaraNodeAssignment_2.SetVariables_#1
  &SystemState
  &self:SystemSpawnScript.SystemScriptSource.SystemScriptGraph.NiagaraNodeAssignment_1.SetVariables_#2
  &self:SystemSpawnScript.SystemScriptSource.SystemScriptGraph.NiagaraNodeAssignment_3.SetVariables_#3
 ]
}
emitters {
 Box_burst_pink {
  emitterUpdateScript [
   &EmitterState(Life Cycle Mode=::NewEnumerator1, Loop Behavior=::NewEnumerator1, Loop Duration=0.2, UseLoopDelay=false)
   &SpawnBurst_Instantaneous(
    Spawn Count=&MakeCustomIntFromBool(Bool=$User.Is Swing Left, True INT=0, False INT=6)
    Use Spawn Probability=false
    UseLoopCountLimit=false
   )
  ]
  particleSpawnScript [
   &InitializeParticle(
    Color=&MakeCustomLinearColorFromBool(Bool=&RandomBool(), True Color=<rgba 1 0.5746 0.2126 1>, False Color=<rgba 1 1 1 1>)
    Lifetime Max=0.22
    Lifetime Min=0.15
    Lifetime Mode=::NewEnumerator1
    Mesh Scale Max=<4 3 3>
    Mesh Scale Min=<1.5 1.5 1.5>
    Mesh Scale Mode=::NewEnumerator2
    UsePositionOffset=false
    Write Material Random=true
   )
   &ShapeLocation(Box Size=<50 50 50>, MaskSpawnGroup=false, Override Seed=false, Shape Primitive=::NewEnumerator2, Surface Only=false, UseOwnerScale=true)
   &AddVelocity(
    Module.Use Falloff Distance=false
    Override Seed=false
    Velocity Mode=::NewEnumerator1
    Velocity Speed=&RandomRangeFloat(Minimum=100, Maximum=100)
   )
   &CameraOffset(Camera Offset Amount=10)
  ]
  particleUpdateScript [
   &ParticleState()
   &ScaleMeshSize(Scale Factor=&Multiply_VectorByFloat(
     Vector=<1 1 1>
     Float=&Multiply_Float(
      A=0.1
      B=&FloatFromCurve(
       FloatCurve=DataInterface() {propertyValues {Curve {keys [
           {arriveTangent=4.1785, time=0, value=0}
           <at 0.5 1.1>
           {arriveTangent=-0.0405, interpMode=RCIM_Cubic, leaveTangent=-0.0405, tangentMode=RCTM_User, time=1, value=1.5}
          ]}}}
       Scale Curve=&Multiply_Float(
        A=1
        B=&ClampFloat(
         Float=&FN_LOD_Distance_Remap(Start Distance=200, End Distance=400, Minimum Output=0.5, Maximum Output=1)
         Min=0.5
        )
       )
      )
     )
    ))
   &Drag(Drag=4, Use Linear Drag=true, Use Rotational Drag=true)
   &SolveForcesAndVelocity(Clamp Velocity=false, Limit Acceleration=false)
   &ScaleColor(
    Scale Alpha=&Multiply_Float(
     A=&FloatFromCurve(
      FloatCurve=DataInterface() {propertyValues {Curve {keys [<at 0 1>,<at 1 0>]}}}
      Scale Curve=&RandomRangeFloat(Minimum=0.4, Maximum=1)
     )
     B=3
    )
    Scale RGB=&Multiply_VectorByFloat(
     Vector=&MakeVectorFromLinearColorRGB(Color=&ColorFromCurve(ColorCurve=DataInterface() {propertyValues {
         AlphaCurve {keys []}
         BlueCurve {keys [<at 0.026 0>, <at 0.2869 0>, <at 0.4134 0>, {interpMode=RCIM_Cubic, leaveTangent=12.6172, time=0.7107, value=0}]}
         GreenCurve {keys [<at 0.026 0.0507>, <at 0.2869 2.0314>, <at 0.4134 0.0442>, {interpMode=RCIM_Cubic, leaveTangent=2.394, time=0.7107, value=0.3634}]}
         RedCurve {keys [<at 0.026 0.126>, <at 0.2869 5>, <at 0.4134 5>, {interpMode=RCIM_Cubic, leaveTangent=-12.6172, time=0.7107, value=1}]}
        }}))
     Float=&FloatFromCurve(
      FloatCurve=DataInterface() {propertyValues {Curve {keys [{arriveTangent=-2.0684, interpMode=RCIM_Cubic, leaveTangent=-2.0684, tangentMode=RCTM_User, time=0, value=2}, <at 1 0>]}}}
      Scale Curve=&Multiply_Float(A=3, B=$System.TOD)
     )
    )
    ScaleA=true
    ScaleRGB=true
   )
  ]
  rendererClasses [&NiagaraMeshRendererProperties]
  renderers [{rendererClass=&NiagaraMeshRendererProperties, rendererIndex=0}]
 }
 Box_rate_pink {
  emitterUpdateScript [
   &EmitterState(Life Cycle Mode=::NewEnumerator1, Loop Behavior=::NewEnumerator1, Loop Duration=0.2, UseLoopDelay=true)
   &SpawnRate(
    SpawnRate=&MakeCustomFloatFromBool(Bool=$User.Is Swing Left, True Float=0, False Float=40)
    Use Spawn Probability=false
   )
  ]
  particleSpawnScript [
   &InitializeParticle(
    Color=&MakeCustomLinearColorFromBool(Bool=&RandomBool(), True Color=<rgba 1 0.5746 0.2126 1>, False Color=<rgba 1 1 1 1>)
    Lifetime Max=0.125
    Lifetime Min=0.05
    Lifetime Mode=::NewEnumerator1
    Mesh Scale Max=<4.5 3 3>
    Mesh Scale Min=<1.5 1.5 1.5>
    Mesh Scale Mode=::NewEnumerator2
    UsePositionOffset=false
    Write Material Random=true
   )
   &ShapeLocation(Box Size=<30 30 30>, MaskSpawnGroup=false, Override Seed=false, Shape Primitive=::NewEnumerator2, Surface Only=false, UseOwnerScale=true)
   &CameraOffset(Camera Offset Amount=20)
  ]
  particleUpdateScript [
   &ParticleState()
   &ScaleMeshSize(Scale Factor=&Multiply_VectorByFloat(
     Vector=<1 1 1>
     Float=&Multiply_Float(
      A=0.1
      B=&FloatFromCurve(
       FloatCurve=DataInterface() {propertyValues {Curve {keys [
           {arriveTangent=4.1785, time=0, value=0}
           <at 0.5 1.1>
           {arriveTangent=-0.0405, interpMode=RCIM_Cubic, leaveTangent=-0.0405, tangentMode=RCTM_User, time=1, value=1.5}
          ]}}}
       Scale Curve=&Multiply_Float(
        A=1
        B=&ClampFloat(
         Float=&FN_LOD_Distance_Remap(Start Distance=200, End Distance=400, Minimum Output=0.5, Maximum Output=1)
         Min=0.5
        )
       )
      )
     )
    ))
   &Drag(Drag=4, Use Linear Drag=true, Use Rotational Drag=true)
   &ScaleVelocity(Velocity Scale=&Multiply_VectorByFloat(
     Vector=<1 1 1>
     Float=&FloatFromCurve(FloatCurve=DataInterface() {propertyValues {Curve {keys [
          {interpMode=RCIM_Cubic, tangentMode=RCTM_User, time=0, value=0}
          <at 0.5 1.25>
          <at 0.8 1.5>
          {interpMode=RCIM_Cubic, tangentMode=RCTM_User, time=1, value=0.375}
         ]}}})
    ))
   &SolveForcesAndVelocity(Clamp Velocity=false, Limit Acceleration=false)
   &ScaleColor(
    Scale Alpha=&Multiply_Float(
     A=&FloatFromCurve(
      FloatCurve=DataInterface() {propertyValues {Curve {keys [<at 0 1>,<at 1 0>]}}}
      Scale Curve=&RandomRangeFloat(Minimum=0.4, Maximum=1)
     )
     B=3
    )
    Scale RGB=&Multiply_VectorByFloat(
     Vector=&MakeVectorFromLinearColorRGB(Color=&ColorFromCurve(ColorCurve=DataInterface() {propertyValues {
         AlphaCurve {keys []}
         BlueCurve {keys [<at 0.026 0>, <at 0.2869 0>, <at 0.4134 0>, {interpMode=RCIM_Cubic, leaveTangent=12.6172, time=0.7107, value=0}]}
         GreenCurve {keys [<at 0.026 0.0507>, <at 0.2869 2.0314>, <at 0.4134 0.0442>, {interpMode=RCIM_Cubic, leaveTangent=2.394, time=0.7107, value=0.3634}]}
         RedCurve {keys [<at 0.026 0.126>, <at 0.2869 5>, <at 0.4134 5>, {interpMode=RCIM_Cubic, leaveTangent=-12.6172, time=0.7107, value=1}]}
        }}))
     Float=&FloatFromCurve(
      FloatCurve=DataInterface() {propertyValues {Curve {keys [{arriveTangent=-2.0684, interpMode=RCIM_Cubic, leaveTangent=-2.0684, tangentMode=RCTM_User, time=0, value=2}, <at 1 0>]}}}
      Scale Curve=&Multiply_Float(A=3, B=$System.TOD)
     )
    )
    ScaleA=true
    ScaleRGB=true
   )
  ]
  rendererClasses [&NiagaraMeshRendererProperties]
  renderers [{rendererClass=&NiagaraMeshRendererProperties, rendererIndex=0}]
 }
 Empty {
  bEnabled=false
  emitterUpdateScript [
   &EmitterState()
   &SpawnBurst_Instantaneous(Use Spawn Probability=false, UseLoopCountLimit=false)
  ]
  particleSpawnScript [&InitializeParticle(
    Color Mode=::NewEnumerator0
    Lifetime=5
    Mesh Scale Mode=::NewEnumerator3
    Mesh Uniform Scale=0.5
    UsePositionOffset=false
    Write Material Random=true
   )]
  particleUpdateScript [&ParticleState()]
  rendererClasses [&NiagaraMeshRendererProperties]
  renderers [{rendererClass=&NiagaraMeshRendererProperties, rendererIndex=0}]
 }
 Flare {
  bEnabled=false
  emitterUpdateScript [
   &EmitterState(Life Cycle Mode=::NewEnumerator1, Loop Behavior=::NewEnumerator1, Loop Delay=0.1, Loop Duration=1, UseLoopDelay=false)
   &SpawnBurst_Instantaneous(Spawn Count=1, Use Spawn Probability=false)
  ]
  particleSpawnScript [
   &InitializeParticle(
    Color=<rgba 0 0 0 1>
    Lifetime=0.125
    Sprite Rotation Mode=::NewEnumerator2
    Sprite Size Mode=::NewEnumerator3
    Uniform Sprite Size=50
    UsePositionOffset=false
    Write Material Random=true
   )
   &DynamicMaterialParameters(
    Index 0 Param 1=&RandomRangeFloat(Minimum=0.4, Maximum=0.6)
    Index 0 Param 2=1
    Index 0 Param 3=1
    Index 0 Param 4=1
    Module.Param0WriteEnabled=true
    Module.Param1WriteEnabled=true
    Module.Param2WriteEnabled=true
    Module.Param3WriteEnabled=true
   )
   /SetVariables(Particles.Dark_Glow_CameraOffset=&MakeCustomFloatFromBool(Bool=$User.bIsFrontEndPreview, True Float=1, False Float=50)) {isSetParameters=true}
   &CameraOffset(Camera Offset Amount=&MakeCustomFloatFromBool(Bool=$User.bIsFrontEndPreview, True Float=1, False Float=30))
  ]
  particleUpdateScript [
   /SetVariables(Particles.Dark_Glow=&MakeLinearColorFromVectorAndFloat(
     "Vector (RGB)"=&MakeVectorFromLinearColorRGB(Color=&MakeCustomLinearColorFromBool(
       Bool=$System.VarientBase
       True Color=&MultiplyLinearColor(
        Linear Color=<rgba 0.5647 0.2461 0.5289 1>
        Scale Factor=&MakeVector4_FromFloat(Float=&FNTimeOfDayScalar(Night Scale=0.7, Morning Scale=0.7, Evening Scale=0.7))
       )
       False Color=&MultiplyLinearColor(
        Linear Color=<rgba 0.6867 0.0648 0.2747 1>
        Scale Factor=&MakeVector4_FromFloat(Float=&FNTimeOfDayScalar(Night Scale=0.7, Morning Scale=0.7, Evening Scale=0.7))
       )
      ))
     "Float (Alpha)"=&Multiply_Float(
      A=&Multiply_Float(A=&FNTimeOfDayScalar(Night Scale=0.7, Morning Scale=0.7, Evening Scale=1), B=0.5)
      B=&FloatFromCurve(FloatCurve=DataInterface() {propertyValues {Curve {keys [<at 0 0>,<at 0.2 1>,<at 0.7 0.6>,<at 1 0>]}}})
     )
    )) {isSetParameters=true}
   /SetVariables(Particles.Dark_Glow_Size=&Multiply_Vector2(
     A=&Vector2DFromFloat(Value=220)
     B=&Vector2DFromFloat(Value=&FloatFromCurve(
       FloatCurve=DataInterface() {propertyValues {Curve {keys [<at 0 1.5 RCIM_Cubic>,<at 1 0.5>]}}}
       Scale Curve=&FN_FrontEndFloatModifier(Gameplay=0.6, Front End=0.6)
      ))
    )) {isSetParameters=true}
   &ParticleState()
   &ScaleColor(
    Scale Alpha=&Multiply_Float(
     A=1
     B=&Multiply_Float(
      A=&FNTimeOfDayScalar(Night Scale=1, Morning Scale=0.5, Evening Scale=0.5)
      B=&FloatFromCurve(
       FloatCurve=DataInterface() {propertyValues {Curve {keys [<at 0 0>,<at 0.2 1>,<at 0.8 1>,<at 1 0>]}}}
       Scale Curve=&FNTimeOfDayScalar(Night Scale=1, Morning Scale=0.5, Evening Scale=0.5)
      )
     )
    )
    Scale RGB=<2 2 2>
    ScaleA=true
    ScaleRGB=true
   )
   &ScaleSpriteSize(Scale Factor=&Vector2DFromFloat(Value=&FloatFromCurve(
      FloatCurve=DataInterface() {propertyValues {Curve {keys [<at 0 1.5 RCIM_Cubic>,<at 1 0>]}}}
      Scale Curve=&FN_FrontEndFloatModifier(Gameplay=1, Front End=0.6)
     )))
  ]
  rendererClasses [&NiagaraSpriteRendererProperties]
  renderers [
   {rendererClass=&NiagaraSpriteRendererProperties}
   {rendererClass=&NiagaraSpriteRendererProperties, rendererIndex=1}
  ]
 }
 Flash_Flare_pink {
  emitterUpdateScript [
   &EmitterState(Life Cycle Mode=::NewEnumerator1, Loop Behavior=::NewEnumerator1, Loop Duration=1, UseLoopDelay=false)
   &SpawnBurst_Instantaneous(
    Spawn Count=&MakeCustomIntFromBool(Bool=$User.Is Swing Left, True INT=0, False INT=1)
    Use Spawn Probability=false
   )
  ]
  particleSpawnScript [
   &InitializeParticle(
    Lifetime=0.2
    Sprite Rotation Angle=45
    Sprite Rotation Mode=::NewEnumerator0
    Sprite Size Max=<100 150>
    Sprite Size Min=<130 100>
    Sprite Size Mode=::NewEnumerator2
    UsePositionOffset=false
    Write Material Random=true
   ) {module=InitializeParticle001}
   &CameraOffset(Camera Offset Amount=30)
   &DynamicMaterialParameters(
    Index 0 Param 1=1
    Index 0 Param 2=1
    Module.Param0WriteEnabled=true
    Module.Param1WriteEnabled=true
    Module.Param2WriteEnabled=true
    Module.Param3WriteEnabled=true
   )
   /SetVariables(Particles.Rotation2=-45) {isSetParameters=true}
  ]
  particleUpdateScript [
   &ParticleState()
   &SolveForcesAndVelocity(Clamp Velocity=false, Limit Acceleration=false)
   &ScaleColor(
    Scale Alpha=&Multiply_Float(A=&Lerp_Float(A=1, B=0, Alpha=$Particles.NormalizedAge), B=$System.TOD)
    Scale RGB=&MakeVectorFromLinearColorRGB(Color=&ColorFromCurve(ColorCurve=DataInterface() {propertyValues {
        AlphaCurve {keys []}
        BlueCurve {keys [<at 0.1418 0>,<at 0.4953 0>,<at 0.9888 0>]}
        GreenCurve {keys [<at 0.1418 1>,<at 0.4953 0.402>,<at 0.9888 0.0507>]}
        RedCurve {keys [<at 0.1418 1>,<at 0.4953 1>,<at 0.9888 0.126>]}
       }}))
    ScaleA=true
    ScaleRGB=true
   )
   &ScaleSpriteSize(Scale Factor=&Vector2DFromCurve(
     Vector2Curve=DataInterface() {propertyValues {XCurve {keys [<at 0 1>,<at 1 1>]}, YCurve {keys [<at 0 1>,<at 1 0.3>]}}}
     ScaleCurve=&Multiply_Vector2D_ByFloat(Float=0.2, Vector2D=<1.1 1.1>)
    ))
  ]
  rendererClasses [&NiagaraSpriteRendererProperties]
  renderers [
   {rendererClass=&NiagaraSpriteRendererProperties}
   {rendererClass=&NiagaraSpriteRendererProperties, rendererIndex=1}
  ]
 }
 GlitchedCubes {
  emitterUpdateScript [
   &EmitterState(Life Cycle Mode=::NewEnumerator1, Loop Behavior=::NewEnumerator1, Loop Duration=1, UseLoopDelay=false)
   &SpawnBurst_Instantaneous(
    Spawn Count=&RandomRangeInt(Minimum=6, Maximum=8)
    Use Spawn Probability=false
    UseLoopCountLimit=false
   )
  ]
  particleSpawnScript [
   &InitializeParticle(
    Color=&HueShiftLinearColor(
     Hue=0
     Saturation=4
     Value=0.8
     Linear Color=&ColorFromCurve(
      ColorCurve=DataInterface() {propertyValues {
        AlphaCurve {keys [<at 0 1 RCIM_Cubic>,<at 1 1 RCIM_Cubic>]}
        BlueCurve {keys [<at 0 0 RCIM_Cubic>,<at 1 0 RCIM_Cubic>]}
        GreenCurve {keys [{arriveTangent=-13.0525, interpMode=RCIM_Cubic, leaveTangent=-13.0525, tangentMode=RCTM_User, time=0, value=5}, <at 1 0.55 RCIM_Cubic>]}
        RedCurve {keys [{arriveTangent=-13.0525, interpMode=RCIM_Cubic, leaveTangent=-13.0525, tangentMode=RCTM_User, time=0, value=5}, <at 1 1 RCIM_Cubic>]}
       }}
      CurveIndex=&RandomRangeFloat(Minimum=0, Maximum=1)
     )
    )
    Lifetime Max=0.5
    Lifetime Min=0.2
    Lifetime Mode=::NewEnumerator1
    Sprite Rotation Angle=&SelectFloatFromArray(
     Array Sampling Mode=::NewEnumerator0
     Direct Array Index=?
     Float Selection Array=DataInterface() {propertyValues {FloatData [0,0,0]}}
     Random Seed=0
    )
    Sprite Rotation Mode=::NewEnumerator1
    Sprite Size Mode=::NewEnumerator3
    Uniform Sprite Size=&FloatFromCurve(
     FloatCurve=DataInterface() {propertyValues {Curve {keys [<at 0 1.5>,<at 0.2 1.5>,<at 0.8 4>,<at 1 4>]}}}
     CurveIndex=&RandomRangeFloat(Minimum=0, Maximum=1)
    )
    UsePositionOffset=false
    Write Material Random=true
   )
   &ShapeLocation(
    MaskSpawnGroup=false
    Override Seed=false
    Sphere Radius=&RandomRangeFloat(Minimum=1, Maximum=4)
    UseOwnerScale=true
   )
   &AddVelocityFromPoint(Module.Use Falloff Distance=false, Velocity Strength=&RandomRangeFloat(Minimum=275, Maximum=450))
   &CameraOffset(Camera Offset Amount=10)
  ]
  particleUpdateScript [
   &Drag(Drag=&RandomRangeFloat(Minimum=10, Maximum=14), Use Linear Drag=true, Use Rotational Drag=false)
   &SolveForcesAndVelocity(Clamp Velocity=false, Limit Acceleration=false)
   &ParticleState()
   &ScaleSpriteSize(Uniform Curve Sprite Scale=DataInterface() {propertyValues {Curve {keys [
        {arriveTangent=31.886, interpMode=RCIM_Cubic, leaveTangent=31.886, tangentMode=RCTM_User, time=0, value=0}
        {arriveTangent=-1.56, interpMode=RCIM_Cubic, leaveTangent=-1.56, tangentMode=RCTM_User, time=0.2386, value=2.6353}
        <at 0.6 2.5>
        <at 0.7 4.75>
        {arriveTangent=-9.2705, interpMode=RCIM_Cubic, leaveTangent=-9.2705, time=0.8412, value=2.2823}
        {arriveTangent=-14.3763, interpMode=RCIM_Cubic, leaveTangent=-14.3763, time=0.9275, value=2.6413}
        {arriveTangent=-0.0405, interpMode=RCIM_Cubic, leaveTangent=-0.0405, tangentMode=RCTM_User, time=1, value=0}
       ]}}})
   &DynamicMaterialParameters(
    Index 0 Param 1=0.02
    Module.Param0WriteEnabled=true
    Module.Param1WriteEnabled=true
    Module.Param2WriteEnabled=true
    Module.Param3WriteEnabled=true
   ) {enabled=false}
   &DynamicMaterialParameters(
    Index 0 Param 2=&RandomRangeFloat(Minimum=-1, Maximum=1)
    Index 0 Param 3=&Multiply_Float(
     A=&FloatFromCurve(FloatCurve=DataInterface() {propertyValues {Curve {keys [<at 0 1>,<at 1 0>]}}})
     B=&RandomRangeFloat(Minimum=1.4, Maximum=1)
    )
    Index 0 Param 4=&RandomRangeFloat(Minimum=0.1, Maximum=0.2)
    Index 1 Param 4=&FloatFromCurve(FloatCurve=DataInterface() {propertyValues {Curve {keys [
         {arriveTangent=4.1785, interpMode=RCIM_Cubic, leaveTangent=4.1785, tangentMode=RCTM_User, time=0, value=0}
         {arriveTangent=0.9995, interpMode=RCIM_Cubic, leaveTangent=0.9995, tangentMode=RCTM_User, time=0.2223, value=0.6825}
         {arriveTangent=-0.0405, interpMode=RCIM_Cubic, leaveTangent=-0.0405, tangentMode=RCTM_User, time=0.5, value=1}
        ]}}})
    Module.Param0WriteEnabled=false
    Module.Param0WriteEnabled1=false
    Module.Param1WriteEnabled=true
    Module.Param1WriteEnabled1=false
    Module.Param2WriteEnabled=true
    Module.Param2WriteEnabled1=false
    Module.Param3WriteEnabled=true
    Module.Param3WriteEnabled1=true
    Write Parameter Index 1=true
   ) {module=DynamicMaterialParameters001}
  ]
  rendererClasses [&NiagaraSpriteRendererProperties]
  renderers [
   {rendererClass=&NiagaraSpriteRendererProperties}
   {rendererClass=&NiagaraSpriteRendererProperties, rendererIndex=1}
  ]
 }
 Holographic_center_pink {
  bEnabled=false
  emitterUpdateScript [
   &EmitterState(Life Cycle Mode=::NewEnumerator1, Loop Behavior=::NewEnumerator1, Loop Delay=0.05, Loop Duration=1, UseLoopDelay=true)
   &SpawnBurst_Instantaneous(
    Spawn Count=&MakeCustomIntFromBool(Bool=$User.Is Swing Left, True INT=0, False INT=7)
    Use Spawn Probability=false
    UseLoopCountLimit=false
   )
  ]
  particleSpawnScript [
   &InitializeParticle(
    Lifetime Max=0.2
    Lifetime Min=0.15
    Lifetime Mode=::NewEnumerator1
    Mesh Scale Max=<5 6 6>
    Mesh Scale Min=<3 3 3>
    Mesh Scale Mode=::NewEnumerator2
    UsePositionOffset=false
    Write Material Random=true
   )
   &CameraOffset(Camera Offset Amount=20)
   &ShapeLocation(
    Box Size=&Multiply_VectorByFloat(Vector=<50 50 50>, Float=&FN_FrontEndFloatModifier(Gameplay=1, Front End=0.5))
    MaskSpawnGroup=false
    Override Seed=false
    Shape Primitive=::NewEnumerator2
    Surface Only=false
    UseOwnerScale=true
   )
  ]
  particleUpdateScript [
   &ParticleState()
   &ScaleMeshSize(Scale Factor=&Multiply_VectorByFloat(
     Vector=<1 1 1>
     Float=&Multiply_Float(
      A=0.2
      B=&FloatFromCurve(
       FloatCurve=DataInterface() {propertyValues {Curve {keys [
           {arriveTangent=4.1785, time=0, value=0}
           <at 0.5 1.1>
           {arriveTangent=-0.0405, interpMode=RCIM_Cubic, leaveTangent=-0.0405, tangentMode=RCTM_User, time=1, value=1.5}
          ]}}}
       Scale Curve=&FN_FrontEndFloatModifier(Gameplay=0.75, Front End=0.75)
      )
     )
    ))
   &Drag(Drag=2, Use Linear Drag=true, Use Rotational Drag=true)
   &ScaleVelocity(Velocity Scale=&Multiply_VectorByFloat(
     Vector=<1 1 1>
     Float=&FloatFromCurve(FloatCurve=DataInterface() {propertyValues {Curve {keys [
          {interpMode=RCIM_Cubic, tangentMode=RCTM_User, time=0, value=0}
          <at 0.5 1.25>
          <at 0.8 1.5>
          {interpMode=RCIM_Cubic, tangentMode=RCTM_User, time=1, value=0.375}
         ]}}})
    )) {enabled=false}
   &SolveForcesAndVelocity(Clamp Velocity=false, Limit Acceleration=false)
   &ScaleColor(
    Scale Alpha=&Multiply_Float(
     A=&FloatFromCurve(
      FloatCurve=DataInterface() {propertyValues {Curve {keys [
          {arriveTangent=9.0725, interpMode=RCIM_Cubic, leaveTangent=9.0725, tangentMode=RCTM_User, time=0, value=0}
          {arriveTangent=-2.5249, interpMode=RCIM_Cubic, leaveTangent=-2.5249, tangentMode=RCTM_User, time=0.2, value=0.9847}
          {arriveTangent=-0.0405, interpMode=RCIM_Cubic, leaveTangent=-0.0405, tangentMode=RCTM_User, time=1, value=0}
         ]}}}
      Scale Curve=&RandomRangeFloat(Minimum=0.2, Maximum=1)
     )
     B=0.5
    )
    Scale RGB=&Multiply_VectorByFloat(
     Vector=&MakeVectorFromLinearColorRGB(Color=&ColorFromCurve(ColorCurve=DataInterface() {propertyValues {
         AlphaCurve {keys []}
         BlueCurve {keys [<at 0 0 RCIM_Cubic>,<at 0.1326 0>,<at 0.2869 0>,<at 0.406 0>,<at 0.6236 0>,<at 0.7107 0 RCIM_Cubic>,<at 0.8097 0>]}
         GreenCurve {keys [
           <at 0 5 RCIM_Cubic>
           <at 0.1326 0.0507>
           <at 0.2869 5>
           <at 0.406 0.402>
           <at 0.6236 2.5>
           {arriveTangent=13.4337, interpMode=RCIM_Cubic, leaveTangent=13.4337, time=0.7107, value=0.3634}
           <at 0.8097 5>
          ]}
         RedCurve {keys [<at 0 5 RCIM_Cubic>,<at 0.1326 0.126>,<at 0.2869 5>,<at 0.406 1>,<at 0.6236 5>,<at 0.7107 1 RCIM_Cubic>,<at 0.8097 5>]}
        }}))
     Float=&FloatFromCurve(
      FloatCurve=DataInterface() {propertyValues {Curve {keys [{arriveTangent=-2.0684, interpMode=RCIM_Cubic, leaveTangent=-2.0684, tangentMode=RCTM_User, time=0, value=2}, <at 1 0>]}}}
      Scale Curve=&Multiply_Float(A=$System.TOD, B=&FN_FrontEndFloatModifier(Gameplay=1, Front End=2))
     )
    )
    ScaleA=true
    ScaleRGB=true
   )
  ]
  rendererClasses [&NiagaraMeshRendererProperties]
  renderers [{rendererClass=&NiagaraMeshRendererProperties, rendererIndex=0}]
 }
 Impact_Flare_Variant_01 {
  bEnabled=false
  emitterUpdateScript [
   &EmitterState()
   &SpawnBurst_Instantaneous(Spawn Count=1, Spawn Probability=HlslExpression() {hlslExpression="(User.ColorSwitcher==1.0) ? 1.0 : 0.0"}, Use Spawn Probability=true)
  ]
  particleSpawnScript [
   &InitializeParticle(
    Color=&MakeCustomLinearColorFromBool(Bool=$System.VarientBase, True Color=<rgba 1.9089 0.1771 2 1>, False Color=<rgba 0.6867 0.0648 0.2747 1>)
    Lifetime=0.15
    Mesh Scale Max=<1.05 1.05 1.05>
    Mesh Scale Min=<0.9375 0.9375 0.9375>
    Mesh Scale Mode=::NewEnumerator2
    Sprite Size=<200 175>
    Sprite Size Mode=::NewEnumerator4
    UsePositionOffset=false
    Write Material Random=true
   ) {module=InitializeParticle001}
   &CameraOffset(Camera Offset Amount=50)
   &InitialMeshOrientation(
    Rotation=&RandomRangeVector(Minimum=<-1 -1 -1>, Maximum=<1 1 1>)
    Use Orientation Vector=true
    Use Rotation Vector=true
   ) {module=InitialMeshOrientation001}
   /SetVariables(Particles.Scale=&Multiply_VectorByFloat(Vector=$Particles.Scale, Float=0.66)) {isSetParameters=true}
  ]
  particleUpdateScript [
   &ParticleState()
   &ScaleColor(
    Scale Alpha=&FloatFromCurve(FloatCurve=DataInterface() {propertyValues {Curve {keys [{arriveTangent=-2.6501, interpMode=RCIM_Cubic, leaveTangent=-2.6501, tangentMode=RCTM_User, time=0, value=1}, <at 1 0>]}}})
    ScaleA=true
    ScaleRGB=false
   ) {enabled=false}
   &ScaleMeshSize(Scale Factor=&VectorFromFloat(Value=&FloatFromCurve(
      FloatCurve=DataInterface() {propertyValues {Curve {keys [{arriveTangent=1.5592, interpMode=RCIM_Cubic, leaveTangent=1.5592, tangentMode=RCTM_User, time=0, value=0.35}, <at 0.6 1.25>]}}}
      Scale Curve=&FN_FrontEndFloatModifier(Gameplay=0.75, Front End=0.75)
     )))
   &DynamicMaterialParameters(
    Index 0 Param 1=&FloatFromCurve(FloatCurve=DataInterface() {propertyValues {Curve {keys [<at 0 1 RCIM_Cubic>,<at 0.8 0 RCIM_Cubic>]}}})
    Index 0 Param 2=2
    Index 0 Param 3=1
    Index 0 Param 4=1
    Index 1 Param 1=&FloatFromCurve(FloatCurve=DataInterface() {propertyValues {Curve {keys [
         {arriveTangent=-2.3163, interpMode=RCIM_Cubic, time=0, value=2}
         {arriveTangent=-1.9019, interpMode=RCIM_Cubic, leaveTangent=-1.9019, tangentMode=RCTM_User, time=1, value=0.4}
        ]}}})
    Index 1 Param 2=0.4
    Index 1 Param 3=0.6
    Index 1 Param 4=1
    Index 2 Param 1=&FloatFromCurve(FloatCurve=DataInterface() {propertyValues {Curve {keys [<at 0.46 0>,<at 1 1>]}}})
    Index 2 Param 2=45
    Module.Param0WriteEnabled=true
    Module.Param0WriteEnabled1=true
    Module.Param0WriteEnabled2=true
    Module.Param1WriteEnabled=true
    Module.Param1WriteEnabled1=true
    Module.Param1WriteEnabled2=true
    Module.Param2WriteEnabled=true
    Module.Param2WriteEnabled1=true
    Module.Param2WriteEnabled2=true
    Module.Param3WriteEnabled=true
    Module.Param3WriteEnabled1=true
    Module.Param3WriteEnabled2=true
    Write Parameter Index 1=true
    Write Parameter Index 2=true
   ) {module=DynamicMaterialParameters001}
   &FN_Mesh_Attributes(
    Color=&LerpLinearColors(StartColor=<rgba 0.0105 0.4063 0.5052 1>, EndColor=<rgba 0 0 0 1>, LerpFactor=$User.ColorSwitcher)
    Create Color Attribute=true
    Create Scale Attribute=true
    Scale=&Multiply_VectorByFloat(Vector=$Particles.Scale, Float=1.3)
    Write to Module Namespace=true
   )
  ]
  rendererClasses [&NiagaraMeshRendererProperties]
  renderers [
   {rendererClass=&NiagaraMeshRendererProperties, rendererIndex=0}
   {rendererClass=&NiagaraMeshRendererProperties, rendererIndex=1}
  ]
 }
 OmnidirectionalBurst {
  bEnabled=false
  emitterUpdateScript [
   &EmitterState(Life Cycle Mode=::NewEnumerator1, Loop Behavior=::NewEnumerator1, Loop Duration=1, UseLoopDelay=false)
   &SpawnBurst_Instantaneous(Spawn Count=5, Use Spawn Probability=false, UseLoopCountLimit=false)
  ]
  particleSpawnScript [
   &InitializeParticle(
    Color Maximum=<rgba 0 1 1 1>
    Color Minimum=<rgba 0.9814 0 1 1>
    Color Mode=::NewEnumerator2
    Lifetime Max=0.35
    Lifetime Min=0.1
    Lifetime Mode=::NewEnumerator1
    Mesh Scale Mode=::NewEnumerator1
    Mesh Uniform Scale Max=8
    Mesh Uniform Scale Min=4
    Sprite Rotation Mode=::NewEnumerator2
    Sprite Size Mode=::NewEnumerator1
    Uniform Sprite Size Max=7
    Uniform Sprite Size Min=3
    UsePositionOffset=false
    Write Material Random=true
   )
   &ShapeLocation(MaskSpawnGroup=false, Override Seed=false, Sphere Radius=0, UseOwnerScale=true)
   &AddVelocity(Velocity=&RandomVector(Vector Scale=&RandomRangeFloat(Minimum=300, Maximum=450)))
  ]
  particleUpdateScript [
   &ParticleState()
   &ScaleMeshSize(Scale Factor=&VectorFromFloat(Value=&FloatFromCurve(FloatCurve=DataInterface() {propertyValues {Curve {keys [<at 0.5 1 RCIM_Cubic>, {arriveTangent=-5.1004, interpMode=RCIM_Cubic, leaveTangent=-5.1004, tangentMode=RCTM_User, time=1, value=0}]}}})))
   &ScaleSpriteSize()
   &Drag(Drag=8, Use Linear Drag=true, Use Rotational Drag=true)
   &SolveForcesAndVelocity(Clamp Velocity=false, Limit Acceleration=false)
  ]
  rendererClasses [&NiagaraMeshRendererProperties]
  renderers [{rendererClass=&NiagaraMeshRendererProperties, rendererIndex=0}]
 }
 PixelCircle {
  emitterUpdateScript [
   &EmitterState(
    Enable Distance Culling=true
    Life Cycle Mode=::NewEnumerator1
    Loop Behavior=::NewEnumerator1
    Loop Duration=1
    Max Distance Response=::KillImmediately
    Module.bUseMaxDistance=true
    Module.bUseMinDistance=false
    Scalability Mode=::NewEnumerator1
    UseLoopDelay=false
   )
   &SpawnBurst_Instantaneous(Use Spawn Probability=false, UseLoopCountLimit=false)
   &SpawnBurst_Instantaneous(Use Spawn Probability=false, UseLoopCountLimit=false) {enabled=false, module=SpawnBurst_Instantaneous001}
  ]
  particleSpawnScript [
   &FN_SetSpawnGroup(Number of Spawn Groups=2)
   &InitializeParticle(
    Color=<rgba 4 3 0 1>
    Lifetime=0.3
    Sprite Rotation Angle=&SelectFloatFromArray(
     Array Sampling Mode=::NewEnumerator0
     Direct Array Index=?
     Float Selection Array=DataInterface() {propertyValues {FloatData [0,0.25,0.5,0.75]}}
     Random Seed=0
    )
    Sprite Rotation Mode=::NewEnumerator1
    Sprite Size Mode=::NewEnumerator3
    Uniform Sprite Size=50
    UsePositionOffset=false
    Write Material Random=true
   )
   &FN_PerParticleLOD(Enable SpriteSize Distance Scaling=true, Sprite Size Scale Far=<2.5 2.5>)
   &CameraOffset(Camera Offset Amount=40)
   /SetVariables(Particles.RandomFloat=&UniformRangedFloat(Minimum=0, Maximum=1)) {isSetParameters=true, module=SetVariables_#4}
   /SetVariables(
    Particles.SubImageIndex=&RandomRangeFloat(Minimum=0, Maximum=3)
    Particles.SubImageIndexAlt:NiagaraFloat=&RandomRangeFloat(Minimum=0, Maximum=3)
   ) {isSetParameters=true, module=SetVariables_#5}
  ]
  particleUpdateScript [
   &ParticleState()
   &ScaleColor(
    Linear Color Curve=DataInterface() {propertyValues {
      AlphaCurve {keys [<at 0.1347 1>,<at 0.8992 0>]}
      BlueCurve {keys [<at 0.1978 1>,<at 0.9037 0.7659>]}
      GreenCurve {keys [<at 0.1978 0.125>,<at 0.9037 0.4597>]}
      RedCurve {keys [<at 0.1978 0.5625>,<at 0.9037 1>]}
     }}
    Scale Mode=::NewEnumerator2
   ) {enabled=false}
   &ScaleColor(
    Linear Color Curve=DataInterface() {propertyValues {
      AlphaCurve {keys [<at 0 1>,<at 1 0>]}
      BlueCurve {keys [<at 0.5192 1>,<at 0.9015 0.75>]}
      GreenCurve {keys [<at 0.5192 1>,<at 0.9015 0.75>]}
      RedCurve {keys [<at 0.5192 1>,<at 0.9015 0.75>]}
     }}
    Scale Mode=::NewEnumerator2
   ) {module=ScaleColor002}
   &ScaleColor(
    Scale RGB=&MakeVectorFromLinearColorRGB(Color=&ColorFromCurve(ColorCurve=DataInterface() {propertyValues {
        AlphaCurve {keys [<at 0 1 RCIM_Cubic>,<at 0.5155 0 RCIM_Cubic>]}
        BlueCurve {keys [<at 0.1913 1 RCIM_Cubic>,<at 0.3986 0.6915 RCIM_Cubic>]}
        GreenCurve {keys [<at 0.1913 0.072 RCIM_Cubic>,<at 0.3986 0.5183 RCIM_Cubic>]}
        RedCurve {keys [<at 0.1913 0.6597 RCIM_Cubic>,<at 0.3986 1 RCIM_Cubic>]}
       }}))
    ScaleA=true
    ScaleRGB=true
   ) {enabled=false, module=ScaleColor001}
   &ScaleSpriteSize(
    Uniform Curve Scale=1.1
    Uniform Curve Sprite Scale=DataInterface() {propertyValues {Curve {keys [
        {arriveTangent=1.0149, interpMode=RCIM_Constant, time=0, value=0}
        <at 0.2 0.25 RCIM_Constant>
        <at 0.4 0.6 RCIM_Constant>
        <at 0.6 0.7 RCIM_Constant>
        <at 0.8 0.8>
       ]}}}
   ) {enabled=false, module=ScaleSpriteSize001}
   &ScaleSpriteSize(
    Uniform Curve Scale=&SelectFloatFromArray(
     Direct Array Index=&ReturnExecIndex()
     Float Selection Array=DataInterface() {propertyValues {FloatData [1.1,1.3]}}
    )
    Uniform Curve Sprite Scale=DataInterface() {propertyValues {Curve {keys [
        {arriveTangent=1.0149, interpMode=RCIM_Constant, time=0, value=0}
        <at 0.2 0.3 RCIM_Constant>
        <at 0.4 0.6 RCIM_Constant>
        <at 0.6 0.7 RCIM_Constant>
       ]}}}
   )
   &DynamicMaterialParameters(
    Index 0 Param 1=&FloatFromCurve(FloatCurve=DataInterface() {propertyValues {Curve {keys [<at 0.5 0>,<at 1 1>]}}})
    Module.Param0WriteEnabled=true
    Module.Param1WriteEnabled=true
    Module.Param2WriteEnabled=true
    Module.Param3WriteEnabled=true
   )
  ]
  rendererClasses [&NiagaraSpriteRendererProperties]
  renderers [{rendererClass=&NiagaraSpriteRendererProperties}]
 }
 PixelFlash {
  emitterUpdateScript [
   &EmitterState(
    Enable Distance Culling=true
    Life Cycle Mode=::NewEnumerator1
    Loop Behavior=::NewEnumerator1
    Loop Duration=1
    Max Distance Response=::KillImmediately
    Module.bUseMaxDistance=true
    Module.bUseMinDistance=false
    Scalability Mode=::NewEnumerator1
    UseLoopDelay=false
   )
   &SpawnBurst_Instantaneous(Spawn Count=2, Use Spawn Probability=false, UseLoopCountLimit=false)
  ]
  particleSpawnScript [
   &FN_SetSpawnGroup(Number of Spawn Groups=2)
   &InitializeParticle(
    Color=<rgba 15 1.5 0 1>
    Lifetime=&Lerp_Float(A=0.2, B=0.2, Alpha=&MakeFloatFromInt(Int=$Particles.SpawnGroup))
    Sprite Rotation Angle=&SelectFloatFromArray(
     Array Sampling Mode=::NewEnumerator0
     Direct Array Index=?
     Float Selection Array=DataInterface() {propertyValues {FloatData [0,0.25,0.5,0.75]}}
     Random Seed=0
    )
    Sprite Rotation Mode=::NewEnumerator1
    Sprite Size Mode=::NewEnumerator3
    Sprite UV Mode=::NewEnumerator3
    Uniform Sprite Size=32
    UsePositionOffset=false
    Write Material Random=true
   )
   &FN_PerParticleLOD(Enable SpriteSize Distance Scaling=true, Sprite Size Scale Far=<2.5 2.5>)
   &CameraOffset(Camera Offset Amount=50)
   /SetVariables(Particles.RandomFloat=&UniformRangedFloat(Minimum=0, Maximum=1)) {isSetParameters=true, module=SetVariables_#4}
   /SetVariables(
    Particles.SubImageIndex=&RandomRangeFloat(Minimum=0, Maximum=3)
    Particles.SubImageIndexAlt:NiagaraFloat=&RandomRangeFloat(Minimum=0, Maximum=3)
   ) {enabled=false, isSetParameters=true, module=SetVariables_#5}
   &SubUVAnimation(End Frame=3, SubUV Animation Mode:ENiagara_SubUVLookupMode=::NewEnumerator2)
  ]
  particleUpdateScript [
   &ParticleState()
   &ScaleColor(
    Linear Color Curve=DataInterface() {propertyValues {
      AlphaCurve {keys [<at 0.1347 1>,<at 0.8992 0>]}
      BlueCurve {keys [<at 0.1978 1>,<at 0.9037 0.7659>]}
      GreenCurve {keys [<at 0.1978 0.125>,<at 0.9037 0.4597>]}
      RedCurve {keys [<at 0.1978 0.5625>,<at 0.9037 1>]}
     }}
    Scale Mode=::NewEnumerator2
   ) {enabled=false}
   &ScaleColor(
    Linear Color Curve=DataInterface() {propertyValues {
      AlphaCurve {keys [<at 0.3624 1>,<at 0.8992 0>]}
      BlueCurve {keys [<at 0.1978 1>]}
      GreenCurve {keys [<at 0.1978 1>]}
      RedCurve {keys [<at 0.1978 1>]}
     }}
    Scale Mode=::NewEnumerator2
   ) {module=ScaleColor002}
   &ScaleColor(
    Scale RGB=&MakeVectorFromLinearColorRGB(Color=&ColorFromCurve(ColorCurve=DataInterface() {propertyValues {
        AlphaCurve {keys [<at 0 1 RCIM_Cubic>,<at 0.5155 0 RCIM_Cubic>]}
        BlueCurve {keys [<at 0.1913 1 RCIM_Cubic>,<at 0.3986 0.6915 RCIM_Cubic>]}
        GreenCurve {keys [<at 0.1913 0.072 RCIM_Cubic>,<at 0.3986 0.5183 RCIM_Cubic>]}
        RedCurve {keys [<at 0.1913 0.6597 RCIM_Cubic>,<at 0.3986 1 RCIM_Cubic>]}
       }}))
    ScaleA=true
    ScaleRGB=true
   ) {enabled=false, module=ScaleColor001}
   &ScaleSpriteSize(
    Uniform Curve Scale=1.1
    Uniform Curve Sprite Scale=DataInterface() {propertyValues {Curve {keys [{arriveTangent=2.6565, interpMode=RCIM_Cubic, leaveTangent=2.6565, tangentMode=RCTM_User, time=0, value=0}, <at 1 1 RCIM_Cubic>]}}}
   ) {module=ScaleSpriteSize001}
  ]
  rendererClasses [&NiagaraSpriteRendererProperties]
  renderers [{rendererClass=&NiagaraSpriteRendererProperties}]
 }
 Sparks003 {
  bEnabled=false
  emitterUpdateScript [
   &EmitterState()
   &SpawnBurst_Instantaneous(Spawn Count=&RandomRangeInt(Minimum=2, Maximum=5), Use Spawn Probability=false)
   &SpawnBurst_Instantaneous(Spawn Count=&RandomRangeInt(Minimum=8, Maximum=10), Spawn Time=0.05, Use Spawn Probability=false) {module=SpawnBurst_Instantaneous001}
  ]
  particleSpawnScript [
   /SetVariables(Particles.Random=&RandomRangeFloat(Minimum=0, Maximum=1)) {isSetParameters=true}
   &InitializeParticle(
    Color=&MakeCustomLinearColorFromBool(
     Bool=$System.VarientBase
     True Color=<rgba 0.6172 0.2346 0.3864 1>
     False Color=&SelectLinearColorFromArray(
      Array Sampling Mode=::NewEnumerator1
      Direct Array Index=&RandomRangeInt(Minimum=0, Maximum=1)
      Color Selection Array=DataInterface() {propertyValues {ColorData [<rgba 0.7011 0.0685 0.2874 1>, <rgba 0.0612 0.7157 0.8228 1>]}}
      Random Seed=?
     )
    )
    Lifetime=&Lerp_Float(A=0.1, B=0.125, Alpha=$Particles.Random)
    Mass Mode=::NewEnumerator1
    Mesh Scale=&Multiply_VectorByFloat(
     Vector=&Multiply_Vector(A=&Lerp_Vector(A=<0.2 0.2 0.2>, B=<0.6 0.6 0.6>, Alpha=$Particles.Random), B=<0.8 0.8 0.4>)
     Float=&FN_FrontEndFloatModifier(Gameplay=0.5, Front End=0.5)
    )
    Mesh Scale Mode=::NewEnumerator4
    Sprite Size=&RandomRangeVector2D(Minimum=<5 40>, Maximum=<10 65>)
    Sprite Size Mode=::NewEnumerator4
    UsePositionOffset=false
    Write Material Random=true
   ) {module=InitializeParticle001}
   &CameraOffset(Camera Offset Amount=30)
   &SphereLocation(
    MaskSpawnGroup=false
    Override Seed=false
    Sphere Radius=1
    Surface Only=false
    UseNonUniformScale=false
    UseOffset=false
    UseSphereOrientation=false
   )
   &AddVelocity(
    Scale Added Velocity=&Multiply_VectorByFloat(Vector=<1 1 1>, Float=&Lerp_Float(A=2, B=3, Alpha=$Particles.Random))
    Velocity=$Output.SphereLocation.SphereNormal
   )
   &DynamicMaterialParameters(
    Index 0 Param 2=&FloatFromCurve(
     FloatCurve=DataInterface() {propertyValues {Curve {keys [<at 0 0 RCIM_Cubic>, {arriveTangent=1, interpMode=RCIM_Cubic, leaveTangent=1, time=0.7035, value=0.3698}, <at 1 1>]}}}
     CurveIndex=&RandomRangeFloat(Minimum=0, Maximum=1)
    )
    Module.Param0WriteEnabled=true
    Module.Param1WriteEnabled=true
    Module.Param2WriteEnabled=true
    Module.Param3WriteEnabled=true
   ) {module=DynamicMaterialParameters001}
   &InitialMeshOrientation(
    Orientation Vector=<1 1 1>
    Rotation=&RandomRangeVector(Minimum=<-1 -1 -1>, Maximum=<1 1 1>)
    Use Orientation Vector=true
    Use Rotation Vector=true
   )
   /SetVariables(Particles.Scale=&Multiply_VectorByFloat(Vector=$Particles.Scale, Float=0.66)) {isSetParameters=true}
  ]
  particleUpdateScript [
   &ParticleState()
   &SolveForcesAndVelocity(Clamp Velocity=false, Limit Acceleration=false)
   &DynamicMaterialParameters(
    Index 0 Param 1=&FloatFromCurve(FloatCurve=DataInterface() {propertyValues {Curve {keys [<at 0 0>,<at 1 1>]}}})
    Module.Param0WriteEnabled=true
    Module.Param1WriteEnabled=false
    Module.Param2WriteEnabled=true
    Module.Param3WriteEnabled=true
   )
  ]
  rendererClasses [&NiagaraSpriteRendererProperties, &NiagaraMeshRendererProperties]
  renderers [
   {rendererClass=&NiagaraSpriteRendererProperties}
   {rendererClass=&NiagaraMeshRendererProperties, rendererIndex=1}
  ]
 }
}
summary {
 systemName=NS_Emote_IndieBucket_Impact
 emitters [
  {bEnabled=false, emitterName=OmnidirectionalBurst, rendererClasses [&NiagaraMeshRendererProperties]}
  {bEnabled=false, emitterName=Flare, rendererClasses [&NiagaraSpriteRendererProperties]}
  {
   bEnabled=false
   emitterName=Sparks003
   rendererClasses [&NiagaraSpriteRendererProperties, &NiagaraMeshRendererProperties]
  }
  {bEnabled=false, emitterName=Impact_Flare_Variant_01, rendererClasses [&NiagaraMeshRendererProperties]}
  {emitterName=Box_rate_pink, rendererClasses [&NiagaraMeshRendererProperties]}
  {emitterName=Box_burst_pink, rendererClasses [&NiagaraMeshRendererProperties]}
  {emitterName=PixelCircle, rendererClasses [&NiagaraSpriteRendererProperties]}
  {emitterName=GlitchedCubes, rendererClasses [&NiagaraSpriteRendererProperties]}
  {emitterName=PixelFlash, rendererClasses [&NiagaraSpriteRendererProperties]}
  {bEnabled=false, emitterName=Empty, rendererClasses [&NiagaraMeshRendererProperties]}
  {bEnabled=false, emitterName=Holographic_center_pink, rendererClasses [&NiagaraMeshRendererProperties]}
  {emitterName=Flash_Flare_pink, rendererClasses [&NiagaraSpriteRendererProperties]}
 ]
 userVariables [
  User.bIsFrontEnd {type=NiagaraBool}
  User.bIsFrontEndPreview {type=NiagaraBool}
  User.ColorSwitcher {type=NiagaraFloat, defaultValue=1, description=""}
  User.Is Swing Left {type=NiagaraBool}
 ]
}
