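# Niagara ribbon trail for the IndieBucket stick emote, drawn from bones or sockets on the user-supplied skeletal mesh.
# Round_Trail spawns two short-lived anim trails tinted green and blue by spawn group, rendered by two ribbon renderers.
# Curves drive alpha, ribbon width and material parameters, partly over emitter loop age, while a time-of-day scalar dims color at night.
# A disabled Outline emitter picks black or a color gradient using System.VarientBase.
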NiagaraSystem /BRCosmetics/Effects/Fort_Effects/Emotes/IndieBucket/NS_Emote_Stick_IndieBucket_Trail
dependencies {
 usedDataInterfaces [&NiagaraDataInterfaceCurve, &NiagaraDataInterfaceColorCurve]
 usedDynamicInputs [&SetBoolByFloatComparison]
 usedModules [
  &self:SystemSpawnScript.SystemScriptSource.SystemScriptGraph.NiagaraNodeAssignment_0.SetVariables_#1
  &self:SystemSpawnScript.SystemScriptSource.SystemScriptGraph.NiagaraNodeAssignment_1.SetVariables_#2
  &self:SystemSpawnScript.SystemScriptSource.SystemScriptGraph.NiagaraNodeAssignment_2.SetVariables_#3
  &SystemState
 ]
}
emitters {
 Outline {
  bEnabled=false
  emitterUpdateScript [
   &EmitterState(Life Cycle Mode=::NewEnumerator1, Loop Behavior=::NewEnumerator1, Loop Duration=1, UseLoopDelay=false)
   &FN_SpawnAnimTrail()
   /SetVariables(Emitter.Random:NiagaraFloat=&RandomRangeFloat(Minimum=0, Maximum=1)) {isSetParameters=true, module=SetVariables_#4}
  ]
  particleSpawnScript [
   &FN_InitializeAnimTrailBoneIndecies()
   &FN_InitializeAnimTrail(Color=<rgba 1 1 1 0.5>, Lifetime=0.12)
   &DynamicMaterialParameters(
    Index 0 Param 4=$Emitter.Random
    Module.Param0WriteEnabled=false
    Module.Param1WriteEnabled=false
    Module.Param2WriteEnabled=false
    Module.Param3WriteEnabled=true
   ) {module=DynamicMaterialParameters001}
  ]
  particleUpdateScript [
   &FN_UpdateAnimTrail(SkeletalMesh=$User.SkeletalMesh) {module=FN_UpdateAnimTrail001}
   /SetVariables(Particles.OutlineColor:LinearColor=&MakeCustomLinearColorFromBool(
     Bool=$System.VarientBase
     True Color=<rgba 0 0 0 0.9>
     False Color=&ColorFromCurve(ColorCurve=DataInterface() {propertyValues {
        AlphaCurve {keys [<at 0 1 RCIM_Cubic>]}
        BlueCurve {keys [<at 0.1051 0.2635 RCIM_Cubic>,<at 0.4991 0>,<at 0.8393 0.4224 RCIM_Cubic>]}
        GreenCurve {keys [<at 0.1051 0.2246 RCIM_Cubic>,<at 0.4991 0.8112>,<at 0.8393 0.1124 RCIM_Cubic>]}
        RedCurve {keys [<at 0.1051 0.031 RCIM_Cubic>,<at 0.4991 0.8802>,<at 0.8393 0.5135 RCIM_Cubic>]}
       }})
    )) {isSetParameters=true}
   &SolveForcesAndVelocity(Clamp Velocity=false, Limit Acceleration=false)
   &ScaleColor(
    Scale Alpha=&Multiply_Float(
     A=&FloatFromCurve(FloatCurve=DataInterface() {propertyValues {Curve {keys [
          {arriveTangent=-1.8862, interpMode=RCIM_Cubic, leaveTangent=-1.8862, tangentMode=RCTM_User, time=0.2, value=1.25}
          {interpMode=RCIM_Cubic, tangentMode=RCTM_User, time=1, value=0}
         ]}}})
     B=&Lerp_Float(
      A=1
      B=&FloatFromCurve(FloatCurve=DataInterface() {propertyValues {Curve {keys [
           {arriveTangent=0.6814, interpMode=RCIM_Cubic, leaveTangent=0.6814, tangentMode=RCTM_User, time=0, value=0.1}
           {arriveTangent=2.5, interpMode=RCIM_Cubic, leaveTangent=2.5, time=0.15, value=0.3}
           {arriveTangent=1.5958, time=0.2, value=0.6}
           {arriveTangent=0.0224, interpMode=RCIM_Cubic, leaveTangent=0.0224, tangentMode=RCTM_User, time=1, value=1}
          ]}}})
      Alpha=&FloatFromCurve(
       FloatCurve=DataInterface() {propertyValues {Curve {keys [<at 0 0 RCIM_Constant>,<at 0.4 1 RCIM_Constant>]}}}
       CurveIndex=$Emitter.NormalizedLoopAge
      )
     )
    )
    Scale RGB=&Multiply_Vector(A=<1 1 1>, B=<1 1 1>)
    ScaleA=true
    ScaleRGB=true
   ) {module=ScaleColor001}
   &ScaleRibbonWidth(Ribbon Width Scale=&FloatFromCurve(
     FloatCurve=DataInterface() {propertyValues {Curve {keys [{arriveTangent=-1.7104, interpMode=RCIM_Cubic, leaveTangent=-1.7104, tangentMode=RCTM_User, time=0.4, value=1}, <at 1 0.5 RCIM_Cubic>]}}}
     CurveIndex=$Emitter.NormalizedLoopAge
    ))
   /SetVariables(Particles.OutlineWidth:NiagaraFloat=&Multiply_Float(A=$Particles.RibbonWidth, B=1)) {isSetParameters=true}
   /SetVariables(Particles.OutlineColor:LinearColor=&MultiplyLinearColor(
     Linear Color=$Particles.OutlineColor
     Scale Factor=&MakeVector4(X=3, Y=3, Z=3, W=&MakeCustomFloatFromBool(Bool=$System.VarientBase, True Float=1, False Float=5))
    )) {enabled=false, isSetParameters=true}
  ]
  rendererClasses [&NiagaraRibbonRendererProperties]
  renderers [{rendererClass=&NiagaraRibbonRendererProperties}]
 }
 Round_Trail {
  emitterUpdateScript [
   &EmitterState()
   &FN_SpawnAnimTrail(Number of Anim Trails=2)
   /SetVariables(Emitter.Random:NiagaraFloat=&RandomRangeFloat(Minimum=0, Maximum=0.1)) {isSetParameters=true, module=SetVariables_#4}
  ]
  particleSpawnScript [
   &FN_InitializeAnimTrailBoneIndecies(AnimTrail Selection=1, First Bone or Socket Index=2, Second Bone or Socket Index=3)
   &FN_InitializeAnimTrailBoneIndecies() {module=FN_InitializeAnimTrailBoneIndecies001}
   &FN_InitializeAnimTrail(
    Color=&MaskLinearColorBySpawnGroup(Masked Linear Color=<rgba 0.2 1 0 0.5>, Passthrough Linear Color=<rgba 0 0.2833 1 1>)
    Lifetime=0.2
   )
   &DynamicMaterialParameters(
    Index 0 Param 4=$Emitter.Random
    Module.Param0WriteEnabled=false
    Module.Param1WriteEnabled=false
    Module.Param2WriteEnabled=false
    Module.Param3WriteEnabled=true
   ) {module=DynamicMaterialParameters001}
  ]
  particleUpdateScript [
   &FN_UpdateAnimTrail(SkeletalMesh=$User.SkeletalMesh) {module=FN_UpdateAnimTrail001}
   &SolveForcesAndVelocity(Clamp Velocity=false, Limit Acceleration=false)
   &ScaleColor(
    Scale Alpha=&Multiply_Float(
     A=&FloatFromCurve(FloatCurve=DataInterface() {propertyValues {Curve {keys [
          {arriveTangent=-1.8862, interpMode=RCIM_Cubic, leaveTangent=-1.8862, tangentMode=RCTM_User, time=0.2, value=1.25}
          {interpMode=RCIM_Cubic, tangentMode=RCTM_User, time=1, value=0}
         ]}}})
     B=&Lerp_Float(
      A=1
      B=&FloatFromCurve(FloatCurve=DataInterface() {propertyValues {Curve {keys [
           {arriveTangent=0.6814, interpMode=RCIM_Cubic, leaveTangent=0.6814, tangentMode=RCTM_User, time=0, value=0.1}
           {arriveTangent=2.5, interpMode=RCIM_Cubic, leaveTangent=2.5, time=0.15, value=0.3}
           {arriveTangent=1.5958, time=0.2, value=0.6}
           {arriveTangent=0.0224, interpMode=RCIM_Cubic, leaveTangent=0.0224, tangentMode=RCTM_User, time=1, value=1}
          ]}}})
      Alpha=&FloatFromCurve(
       FloatCurve=DataInterface() {propertyValues {Curve {keys [<at 0 0 RCIM_Constant>,<at 0.4 1 RCIM_Constant>]}}}
       CurveIndex=$Emitter.NormalizedLoopAge
      )
     )
    )
    Scale RGB=&Multiply_VectorByFloat(Vector=<1 1 1>, Float=&FNTimeOfDayScalar(Night Scale=0.33, Morning Scale=1, Evening Scale=1))
    ScaleA=true
    ScaleRGB=true
   ) {module=ScaleColor001}
   &ScaleColor(
    Scale Alpha=&FloatFromCurve(FloatCurve=DataInterface() {propertyValues {Curve {keys [
         {arriveTangent=0.6576, interpMode=RCIM_Cubic, leaveTangent=0.6576, tangentMode=RCTM_User, time=0, value=0}
         {arriveTangent=1.9656, interpMode=RCIM_Cubic, leaveTangent=1.9656, tangentMode=RCTM_User, time=0.15, value=0.6228}
         {arriveTangent=0.0224, interpMode=RCIM_Cubic, leaveTangent=0.0224, tangentMode=RCTM_User, time=1, value=1}
        ]}}})
    ScaleA=true
    ScaleRGB=true
   ) {enabled=false}
   &ScaleRibbonWidth(Ribbon Width Scale=&FloatFromCurve(
     FloatCurve=DataInterface() {propertyValues {Curve {keys [{arriveTangent=-1.7104, interpMode=RCIM_Cubic, leaveTangent=-1.7104, tangentMode=RCTM_User, time=0.4, value=1}, <at 1 0.5 RCIM_Cubic>]}}}
     CurveIndex=$Emitter.NormalizedLoopAge
    ))
   &DynamicMaterialParameters(
    Index 0 Param 3=&FloatFromCurve(
     FloatCurve=DataInterface() {propertyValues {Curve {keys [<at 0.25 0 RCIM_Cubic>,<at 0.4 1 RCIM_Cubic>]}}}
     CurveIndex=$Emitter.NormalizedLoopAge
    )
    Module.Param0WriteEnabled=false
    Module.Param1WriteEnabled=false
    Module.Param2WriteEnabled=true
    Module.Param3WriteEnabled=false
   )
   /SetVariables(Particles.Color=&MultiplyLinearColor(
     Linear Color=$Particles.Color
     Scale Factor=&MakeVector4(X=1, Y=1, Z=1, W=&MakeCustomFloatFromBool(Bool=$System.VarientBase, True Float=1, False Float=2))
    )) {enabled=false, isSetParameters=true}
  ]
  rendererClasses [&NiagaraRibbonRendererProperties]
  renderers [
   {rendererClass=&NiagaraRibbonRendererProperties}
   {rendererClass=&NiagaraRibbonRendererProperties, rendererIndex=1}
  ]
 }
}
summary {
 systemName=NS_Emote_Stick_IndieBucket_Trail
 emitters [
  {emitterName=Round_Trail, rendererClasses [&NiagaraRibbonRendererProperties]}
  {bEnabled=false, emitterName=Outline, rendererClasses [&NiagaraRibbonRendererProperties]}
 ]
 userVariables [
  User.ColorSwitcher {type=NiagaraFloat, defaultValue=0, description=""}
  User.MeshColors {
   type=NiagaraDataInterfaceArrayColor
   description=""
   defaultValue {
    struct=NiagaraExt_VariableValue_DataInterface
    value {
     dataInterface=&self:NiagaraDataInterfaceArrayColor_0
     dataInterfaceClass=&NiagaraDataInterfaceArrayColor
    }
   }
  }
  User.MeshColors_Variant_01 {
   type=NiagaraDataInterfaceArrayColor
   description=""
   defaultValue {
    struct=NiagaraExt_VariableValue_DataInterface
    value {
     dataInterface=&self:NiagaraDataInterfaceArrayColor_1
     dataInterfaceClass=&NiagaraDataInterfaceArrayColor
    }
   }
  }
  User.RibbonIDOffset {type=NiagaraInt32}
  User.SkeletalMesh {
   type=NiagaraDataInterfaceSkeletalMesh
   defaultValue {
    struct=NiagaraExt_VariableValue_DataInterface
    value {dataInterface=&self:NiagaraDataInterfaceSkeletalMesh_2}
   }
  }
  User.VariantSwitch {type=NiagaraFloat, defaultValue=0, description=""}
 ]
}
